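# Material function for the Endit universal vehicle skin, with fire textures as defaults, outputting material attributes.
# Layers diffuse, Fresnel, four RGBA-mask channels and an optional logo into base color and pulsing emissive, with Finish falloff shading base color.
# Each mask channel supports panning, rotation, noise and normal UV warping, noise erosion, selectable colors and hue-shift multichrome.
# Packed-map roughness, metallic and AO, plus normals, flatten or revert to defaults under the logo.
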MaterialFunction /VehicleCosmetics/UniversalSkins/Endit/MF_Skin_Endit
graph {nodes [
  Constant() {r=1}
  n1 = SetMaterialAttributes!(
   MaterialAttributes
   BaseColor=NamedRerouteUsage()
   Metallic=NamedRerouteUsage()
   Roughness=NamedRerouteUsage()
   EmissiveColor=NamedRerouteUsage()
   Normal=NamedRerouteUsage()
   AmbientOcclusion=NamedRerouteUsage()
  ) {attributeSetTypes [
    #1
    #2
    #3
    #4
    #5
    #6
   ]}
  FunctionOutput(None=@n1)
  LinearInterpolate(
   A=Constant3Vector() {constant=<rgba 0 0 1 0>}
   B=NamedRerouteUsage()
   Alpha=ScalarParameter() {defaultValue=1, group=Fresnel, parameterName=NormalInfluence_Fresnel}
  )
  n2 = TextureSampleParameter2D(UVs=StaticSwitchParameter(True=NamedRerouteUsage(), False=TextureCoordinate()) {defaultValue=true, group=BaseTextures, parameterName=DiffuseMap_UVSet2}) {
   group=BaseTextures
   parameterName=DiffuseAlphaMap
   samplerSource=SSM_Wrap_WorldGroupSettings
   texture=&/T_endit_Diff_Fire1
  }
  n3 = VectorParameter() {group=Colors, parameterName=BGColor, defaultValue=<rgba 0 0 0 1>}
  n4 = StaticSwitchParameter(True=@n2.A, False=Constant()) {defaultValue=true, group=Colors, parameterName=UseBaseColorMap}
  n5 = StaticSwitchParameter(
   True=LinearInterpolate(A=@n2.RGB, B=LinearInterpolate(A=@n3.RGB, B=@n2.RGB, Alpha=@n4), Alpha=ScalarParameter() {group=Colors, parameterName=BGColorOpacity})
   False=@n3.RGB
  ) {defaultValue=true, group=Colors, parameterName=UseBaseColorMap}
  n6 = NamedRerouteDeclaration(None=@n4) {name=DiffuseAlpha, nodeColor=<rgba 1 0.7734 0 1>}
  n7 = NamedRerouteUsage()
  n8 = LinearInterpolate(A=NamedRerouteUsage(), B=@n5, Alpha=@n6)
  n9 = StaticSwitchParameter(
   True=LinearInterpolate(A=NamedRerouteUsage(), B=@n5, Alpha=@n6)
   False=StaticSwitchParameter(
    True=LinearInterpolate(A=@n7, B=@n5, Alpha=@n6)
    False=StaticSwitchParameter(True=@n8, False=@n5) {group=Colors, parameterName=DiffuseAlphaMap_UseTeamColorAsBG}
   ) {group=Colors, parameterName=DiffuseAlphaMap_UseCustomColorAsBG}
  ) {group=Colors, parameterName=DiffuseAlphaMap_UsePaintColorAsBG}
  n10 = Multiply(A=@n9, B=Constant() {r=1}) {desc=Diffuse}
  n11 = NamedRerouteUsage()
  n12 = NamedRerouteUsage()
  n13 = StaticSwitchParameter(
   True=VectorParameter() {group=Fresnel, parameterName=Color_Fresnel, defaultValue=<rgba 1 0 0.0976 1>}.RGB
   False=StaticSwitchParameter(True=NamedRerouteUsage(), False=StaticSwitchParameter(True=@n11, False=@n12) {group=Fresnel, parameterName=UseCustomColor_Fresnel}) {group=Fresnel, parameterName=UsePaintColor_Fresnel}
  ) {group=Fresnel, parameterName=UseUniqueColor_Fresnel}
  n14 = TextureCoordinate() {coordinateIndex=1}
  n15 = ScalarParameter() {defaultValue=1, group=Fresnel, parameterName=MultichromeScale_fresnel}
  n16 = ScalarParameter() {defaultValue=0.2, group=Fresnel, parameterName=MultichromePanSpeed_Fresnel}
  n17 = NamedRerouteUsage()
  n18 = ComponentMask(None=Panner(Coordinate=Multiply(A=@n14, B=@n15), Time=Multiply(A=@n16, B=@n17)) {speedY=1}) {g=true}
  n19 = StaticSwitchParameter(
   True=LinearInterpolate(
    A=@n13
    B=MaterialFunctionCall(Hue Shift Percentage=@n18, Texture=@n13) {materialFunction=&HueShift}.Result
    Alpha=ScalarParameter() {defaultValue=1, group=Fresnel, parameterName=MultichromeColor_Fresnel}
   )
   False=@n13
  ) {group=Fresnel, parameterName=UseMultichromeColor_Fresnel}
  n20 = DotProduct(A=Transform(None=CameraVectorWS()) {transformSourceType=TRANSFORMSOURCE_World, transformType=TRANSFORM_Tangent}, B=NamedRerouteUsage())
  n21 = Multiply(
   A=Power(Base=OneMinus(None=@n20), Exp=ScalarParameter() {defaultValue=5, group=Fresnel, parameterName=Fresnel_Power}) {constExponent=3}
   B=NamedRerouteUsage()
  )
  n22 = Multiply(A=@n21, B=ScalarParameter() {group=Fresnel, parameterName=DiffuseIntensity_Fresnel})
  n23 = NamedRerouteUsage()
  n24 = NamedRerouteUsage()
  n25 = StaticSwitchParameter(
   True=VectorParameter() {group=AlphaChannel, parameterName=Color_A, defaultValue=<rgba 1 1 1 1>}.RGB
   False=StaticSwitchParameter(
    True=NamedRerouteUsage()
    False=StaticSwitchParameter(True=@n23, False=@n24) {defaultValue=true, group=AlphaChannel, parameterName=UseCustomColor_A}
   ) {group=AlphaChannel, parameterName=UsePaintColor_A}
  ) {group=AlphaChannel, parameterName=UseUniqueColor_A}
  n26 = TextureCoordinate() {coordinateIndex=1}
  n27 = ScalarParameter() {defaultValue=1, group=AlphaChannel, parameterName=MultichromeScale_A}
  n28 = ScalarParameter() {defaultValue=1, group=AlphaChannel, parameterName=MultichromePanSpeed_A}
  n29 = NamedRerouteUsage()
  n30 = ComponentMask(None=Panner(Coordinate=Multiply(A=@n26, B=@n27), Time=Multiply(A=@n28, B=@n29)) {speedY=1}) {g=true}
  n31 = StaticSwitchParameter(
   True=LinearInterpolate(
    A=@n25
    B=MaterialFunctionCall(Hue Shift Percentage=@n30, Texture=@n25) {materialFunction=&HueShift}.Result
    Alpha=ScalarParameter() {defaultValue=1, group=AlphaChannel, parameterName=MultichromeColor_A}
   )
   False=@n25
  ) {group=AlphaChannel, parameterName=UseMultichromeColor_A}
  n32 = ScalarParameter() {group=AlphaChannel, parameterName=DiffuseIntensity_A}
  n33 = NamedRerouteUsage()
  n34 = TextureCoordinate()
  n35 = VectorParameter() {desc="RG=XYOffset, BA=XYPanner", group=AlphaChannel, parameterName=OffsetPanner_A}
  n36 = Panner(
   Coordinate=Add(
    A=StaticSwitchParameter(True=@n33, False=@n34) {defaultValue=true, group=AlphaChannel, parameterName=UniversalUVs_A}
    B=AppendVector(A=@n35.R, B=@n35.R)
   )
   Time=NamedRerouteUsage()
   Speed=AppendVector(A=@n35.B, B=@n35.B)
  )
  n37 = Rotator(Coordinate=@n36, Time=Multiply(A=ScalarParameter() {group=AlphaChannel, parameterName=RotatorSpeed_A}, B=NamedRerouteUsage()))
  n38 = ScalarParameter() {defaultValue=0.5, group=AlphaChannel, parameterName=NoiseIntensity_A}
  n39 = VectorParameter() {group=AlphaChannel, parameterName=NoiseTransforms_A, defaultValue=<rgba 1 0.5 0 0.05>}
  n40 = Multiply(
   A=StaticSwitchParameter(True=TextureCoordinate() {coordinateIndex=1}, False=@n34) {defaultValue=true, group=AlphaChannel, parameterName=UniversalNoiseUVs_A}
   B=AppendVector(A=@n39.R, B=@n39.R)
  )
  n41 = Multiply(A=AppendVector(A=@n39.B, B=@n39.B), B=NamedRerouteUsage())
  n42 = ChannelMaskParameter(None=TextureSampleParameter2D(UVs=Add(A=@n40, B=@n41)) {group=BaseTextures, parameterName=NoiseMap, texture=/VehicleCosmetics/Textures/UniversalTextures/T_NPCity_Noise}.RGBA) {group=AlphaChannel, maskChannel=Alpha, parameterName=NoiseChannel_A, defaultValue=<rgba 0 0 0 1>}
  n43 = StaticSwitchParameter(True=ConstantBiasScale(None=@n42.R) {bias=-0.5, scale=2}, False=@n42.R) {defaultValue=true, group=AlphaChannel, parameterName=NoiseCBS_A}
  n44 = Constant()
  n45 = StaticSwitchParameter(True=Multiply(A=Multiply(A=@n38) {constB=0.1}, B=AppendVector(A=@n43, B=@n42.R)), False=@n44) {defaultValue=true, group=AlphaChannel, parameterName=Noise_A}
  n46 = NamedRerouteUsage()
  n47 = StaticSwitchParameter(
   True=Multiply(A=ComponentMask(None=@n46) {g=true, r=true}, B=ScalarParameter() {group=AlphaChannel, parameterName=NormalWarpIntensity_A})
   False=@n44
  ) {group=AlphaChannel, parameterName=NormalWarp_A}
  n48 = TextureSampleParameter2D(UVs=Add(A=StaticSwitchParameter(True=@n37, False=@n36) {group=AlphaChannel, parameterName=Rotator_A}, B=Add(A=@n45, B=@n47))) {group=BaseTextures, parameterName=RGBA_Mask, texture=&/T_endit_RGB_Fire1}
  n49 = LinearInterpolate(
   B=StaticSwitchParameter(True=OneMinus(None=@n42.R), False=@n42.R) {group=AlphaChannel, parameterName=NoiseSubtractInvert_A}
   Alpha=ScalarParameter() {defaultValue=0.1, group=AlphaChannel, parameterName=NoiseSubtractIntensity_A}
  )
  n50 = StaticSwitchParameter(True=Saturate(None=Subtract(A=@n48.A, B=@n49)), False=@n48.A) {group=AlphaChannel, parameterName=NoiseSubtract_A}
  n51 = StaticSwitchParameter(True=LinearInterpolate(A=@n50, Alpha=NamedRerouteUsage()) {constB=0}, False=@n50) {defaultValue=true, group=AlphaChannel, parameterName=AlphaChannelBehindDiffuse}
  n52 = LinearInterpolate(
   A=LinearInterpolate(A=Saturate(None=@n10), B=@n19, Alpha=Saturate(None=@n22)) {constAlpha=1, desc=Diffuse}
   B=@n31
   Alpha=Saturate(None=Multiply(A=@n32, B=@n51))
  ) {constAlpha=1, desc=Diffuse}
  n53 = NamedRerouteUsage()
  n54 = NamedRerouteUsage()
  n55 = StaticSwitchParameter(
   True=VectorParameter() {group=BlueChannel, parameterName=Color_B, defaultValue=<rgba 0 0 1 1>}.RGB
   False=StaticSwitchParameter(
    True=NamedRerouteUsage()
    False=StaticSwitchParameter(True=@n53, False=@n54) {defaultValue=true, group=BlueChannel, parameterName=UseCustomColor_B}
   ) {group=BlueChannel, parameterName=UsePaintColor_B}
  ) {group=BlueChannel, parameterName=UseUniqueColor_B}
  n56 = TextureCoordinate() {coordinateIndex=1}
  n57 = ScalarParameter() {defaultValue=1, group=BlueChannel, parameterName=MultichromeScale_B}
  n58 = ScalarParameter() {defaultValue=1, group=BlueChannel, parameterName=MultichromePanSpeed_B}
  n59 = NamedRerouteUsage()
  n60 = ComponentMask(None=Panner(Coordinate=Multiply(A=@n56, B=@n57), Time=Multiply(A=@n58, B=@n59)) {speedY=1}) {g=true}
  n61 = StaticSwitchParameter(
   True=LinearInterpolate(
    A=@n55
    B=MaterialFunctionCall(Hue Shift Percentage=@n60, Texture=@n55) {materialFunction=&HueShift}.Result
    Alpha=ScalarParameter() {defaultValue=1, group=BlueChannel, parameterName=MultichromeColor_B}
   )
   False=@n55
  ) {group=BlueChannel, parameterName=UseMultichromeColor_B}
  n62 = NamedRerouteUsage()
  n63 = TextureCoordinate()
  n64 = VectorParameter() {desc="RG=XYOffset, BA=XYPanner", group=BlueChannel, parameterName=OffsetPanner_B}
  n65 = Panner(
   Coordinate=Add(
    A=StaticSwitchParameter(True=@n62, False=@n63) {defaultValue=true, group=BlueChannel, parameterName=UniversalUVs_B}
    B=AppendVector(A=@n64.R, B=@n64.R)
   )
   Time=NamedRerouteUsage()
   Speed=AppendVector(A=@n64.B, B=@n64.B)
  )
  n66 = Rotator(Coordinate=@n65, Time=Multiply(A=ScalarParameter() {group=BlueChannel, parameterName=RotatorSpeed_B}, B=NamedRerouteUsage()))
  n67 = ScalarParameter() {defaultValue=0.01, group=BlueChannel, parameterName=NoiseIntensity_B}
  n68 = VectorParameter() {group=BlueChannel, parameterName=NoiseTransforms_B, defaultValue=<rgba 10 10 0 0.05>}
  n69 = Multiply(
   A=StaticSwitchParameter(True=TextureCoordinate() {coordinateIndex=1}, False=@n63) {defaultValue=true, group=BlueChannel, parameterName=UniversalNoiseUVs_B}
   B=AppendVector(A=@n68.R, B=@n68.R)
  )
  n70 = Multiply(A=AppendVector(A=@n68.B, B=@n68.B), B=NamedRerouteUsage())
  n71 = ChannelMaskParameter(None=TextureSampleParameter2D(UVs=Add(A=@n69, B=@n70)) {group=BaseTextures, parameterName=NoiseMap, texture=&T_NPCity_Noise}.RGBA) {group=BlueChannel, maskChannel=Alpha, parameterName=NoiseChannel_B, defaultValue=<rgba 0 0 0 1>}
  n72 = StaticSwitchParameter(True=ConstantBiasScale(None=@n71.R) {bias=-0.5, scale=2}, False=@n71.R) {defaultValue=true, group=BlueChannel, parameterName=NoiseCBS_B}
  n73 = Constant()
  n74 = StaticSwitchParameter(True=Multiply(A=Multiply(A=@n67) {constB=0.1}, B=AppendVector(A=@n72, B=@n71.R)), False=@n73) {defaultValue=true, group=BlueChannel, parameterName=Noise_B}
  n75 = NamedRerouteUsage()
  n76 = StaticSwitchParameter(
   True=Multiply(A=ComponentMask(None=@n75) {g=true, r=true}, B=ScalarParameter() {group=BlueChannel, parameterName=NormalWarpIntensity_B})
   False=@n73
  ) {group=BlueChannel, parameterName=NormalWarp_B}
  n77 = TextureSampleParameter2D(UVs=Add(A=StaticSwitchParameter(True=@n66, False=@n65) {group=BlueChannel, parameterName=Rotator_B}, B=Add(A=@n74, B=@n76))) {group=BaseTextures, parameterName=RGBA_Mask, texture=&T_endit_RGB_Fire1}
  n78 = LinearInterpolate(
   B=StaticSwitchParameter(True=OneMinus(None=@n71.R), False=@n71.R) {group=BlueChannel, parameterName=NoiseSubtractInvert_B}
   Alpha=ScalarParameter() {defaultValue=0.1, group=BlueChannel, parameterName=NoiseSubtractIntensity_B}
  )
  n79 = StaticSwitchParameter(True=Saturate(None=Subtract(A=@n77.B, B=@n78)), False=@n77.B) {group=BlueChannel, parameterName=NoiseSubtract_B}
  n80 = StaticSwitchParameter(True=LinearInterpolate(A=@n79, Alpha=NamedRerouteUsage()) {constB=0}, False=@n79) {group=BlueChannel, parameterName=BlueChannelBehindDiffuse}
  n81 = Saturate(None=Multiply(A=ScalarParameter() {defaultValue=-0.5, group=BlueChannel, parameterName=DiffuseIntensity_B}, B=@n80))
  n82 = NamedRerouteUsage()
  n83 = NamedRerouteUsage()
  n84 = StaticSwitchParameter(
   True=VectorParameter() {group=GreenChannel, parameterName=Color_G, defaultValue=<rgba 0 1 0 1>}.RGB
   False=StaticSwitchParameter(True=NamedRerouteUsage(), False=StaticSwitchParameter(True=@n82, False=@n83) {group=GreenChannel, parameterName=UseCustomColor_G}) {group=GreenChannel, parameterName=UsePaintColor_G}
  ) {group=GreenChannel, parameterName=UseUniqueColor_G}
  n85 = TextureCoordinate() {coordinateIndex=1}
  n86 = ScalarParameter() {defaultValue=1, group=GreenChannel, parameterName=MultichromeScale_G}
  n87 = ScalarParameter() {defaultValue=0.2, group=GreenChannel, parameterName=MultichromePanSpeed_G}
  n88 = NamedRerouteUsage()
  n89 = ComponentMask(None=Panner(Coordinate=Multiply(A=@n85, B=@n86), Time=Multiply(A=@n87, B=@n88)) {speedY=1}) {g=true}
  n90 = StaticSwitchParameter(
   True=LinearInterpolate(
    A=@n84
    B=MaterialFunctionCall(Hue Shift Percentage=@n89, Texture=@n84) {materialFunction=&HueShift}.Result
    Alpha=ScalarParameter() {defaultValue=1, group=GreenChannel, parameterName=MultichromeColor_G}
   )
   False=@n84
  ) {group=GreenChannel, parameterName=UseMultichromeColor_G}
  n91 = NamedRerouteUsage()
  n92 = TextureCoordinate()
  n93 = VectorParameter() {desc="RG=XYOffset, BA=XYPanner", group=GreenChannel, parameterName=OffsetPanner_G}
  n94 = Panner(
   Coordinate=Add(
    A=StaticSwitchParameter(True=@n91, False=@n92) {defaultValue=true, group=GreenChannel, parameterName=UniversalUVs_G}
    B=AppendVector(A=@n93.R, B=@n93.R)
   )
   Time=NamedRerouteUsage()
   Speed=AppendVector(A=@n93.B, B=@n93.B)
  )
  n95 = Rotator(Coordinate=@n94, Time=Multiply(A=ScalarParameter() {group=GreenChannel, parameterName=RotatorSpeed_G}, B=NamedRerouteUsage()))
  n96 = ScalarParameter() {defaultValue=0.5, group=GreenChannel, parameterName=NoiseIntensity_G}
  n97 = VectorParameter() {group=GreenChannel, parameterName=NoiseTransforms_G, defaultValue=<rgba 3 2 0.1 0.5>}
  n98 = Multiply(
   A=StaticSwitchParameter(True=TextureCoordinate() {coordinateIndex=1}, False=@n92) {defaultValue=true, group=GreenChannel, parameterName=UniversalNoiseUVs_G}
   B=AppendVector(A=@n97.R, B=@n97.R)
  )
  n99 = Multiply(A=AppendVector(A=@n97.B, B=@n97.B), B=NamedRerouteUsage())
  n100 = ChannelMaskParameter(None=TextureSampleParameter2D(UVs=Add(A=@n98, B=@n99)) {group=BaseTextures, parameterName=NoiseMap, texture=&T_NPCity_Noise}.RGBA) {group=GreenChannel, parameterName=NoiseChannel_G}
  n101 = StaticSwitchParameter(True=ConstantBiasScale(None=@n100.R) {bias=-0.5, scale=2}, False=@n100.R) {group=GreenChannel, parameterName=NoiseCBS_G}
  n102 = Constant()
  n103 = StaticSwitchParameter(True=Multiply(A=Multiply(A=@n96) {constB=0.1}, B=AppendVector(A=@n101, B=@n100.R)), False=@n102) {defaultValue=true, group=GreenChannel, parameterName=Noise_G}
  n104 = NamedRerouteUsage()
  n105 = StaticSwitchParameter(
   True=Multiply(A=ComponentMask(None=@n104) {g=true, r=true}, B=ScalarParameter() {group=GreenChannel, parameterName=NormalWarpIntensity_G})
   False=@n102
  ) {group=GreenChannel, parameterName=NormalWarp_G}
  n106 = TextureSampleParameter2D(UVs=Add(A=StaticSwitchParameter(True=@n95, False=@n94) {group=GreenChannel, parameterName=Rotator_G}, B=Add(A=@n103, B=@n105))) {group=BaseTextures, parameterName=RGBA_Mask, texture=&T_endit_RGB_Fire1}
  n107 = LinearInterpolate(
   B=StaticSwitchParameter(True=OneMinus(None=@n100.R), False=@n100.R) {defaultValue=true, group=GreenChannel, parameterName=NoiseSubtractInvert_G}
   Alpha=ScalarParameter() {defaultValue=0.1, group=GreenChannel, parameterName=NoiseSubtractIntensity_G}
  )
  n108 = StaticSwitchParameter(True=Saturate(None=Subtract(A=@n106.G, B=@n107)), False=@n106.G) {defaultValue=true, group=GreenChannel, parameterName=NoiseSubtract_G}
  n109 = StaticSwitchParameter(True=LinearInterpolate(A=@n108, Alpha=NamedRerouteUsage()) {constB=0}, False=@n108) {defaultValue=true, group=GreenChannel, parameterName=GreenChannelBehindDiffuse}
  n110 = Multiply(A=ScalarParameter() {group=GreenChannel, parameterName=DiffuseIntensity_G}, B=@n109)
  n111 = NamedRerouteUsage()
  n112 = NamedRerouteUsage()
  n113 = StaticSwitchParameter(
   True=VectorParameter() {group=RedChannel, parameterName=Color_R, defaultValue=<rgba 1 0 0 1>}.RGB
   False=StaticSwitchParameter(True=NamedRerouteUsage(), False=StaticSwitchParameter(True=@n111, False=@n112) {group=RedChannel, parameterName=UseCustomColor_R}) {group=RedChannel, parameterName=UsePaintColor_R}
  ) {group=RedChannel, parameterName=UseUniqueColor_R}
  n114 = TextureCoordinate() {coordinateIndex=1}
  n115 = ScalarParameter() {defaultValue=1, group=RedChannel, parameterName=MultichromeScale_R}
  n116 = ScalarParameter() {defaultValue=0.2, group=RedChannel, parameterName=MultichromePanSpeed_R}
  n117 = NamedRerouteUsage()
  n118 = ComponentMask(None=Panner(Coordinate=Multiply(A=@n114, B=@n115), Time=Multiply(A=@n116, B=@n117)) {speedY=1}) {g=true}
  n119 = StaticSwitchParameter(
   True=LinearInterpolate(
    A=@n113
    B=MaterialFunctionCall(Hue Shift Percentage=@n118, Texture=@n113) {materialFunction=&HueShift}.Result
    Alpha=ScalarParameter() {defaultValue=1, group=RedChannel, parameterName=MultichromeColor_R}
   )
   False=@n113
  ) {group=RedChannel, parameterName=UseMultichromeColor_R}
  n120 = ScalarParameter() {group=RedChannel, parameterName=DiffuseIntensity_R}
  n121 = NamedRerouteUsage()
  n122 = TextureCoordinate()
  n123 = VectorParameter() {desc="RG=XYOffset, BA=XYPanner", group=RedChannel, parameterName=OffsetPanner_R}
  n124 = Panner(
   Coordinate=Add(
    A=StaticSwitchParameter(True=@n121, False=@n122) {defaultValue=true, group=RedChannel, parameterName=UniversalUVs_R}
    B=AppendVector(A=@n123.R, B=@n123.R)
   )
   Time=NamedRerouteUsage()
   Speed=AppendVector(A=@n123.B, B=@n123.B)
  )
  n125 = Rotator(Coordinate=@n124, Time=Multiply(A=ScalarParameter() {group=RedChannel, parameterName=RotatorSpeed_R}, B=NamedRerouteUsage()))
  n126 = ScalarParameter() {defaultValue=0.5, group=RedChannel, parameterName=NoiseIntensity_R}
  n127 = VectorParameter() {group=RedChannel, parameterName=NoiseTransforms_R, defaultValue=<rgba 1.5 2 0.1 0.5>}
  n128 = Multiply(
   A=StaticSwitchParameter(True=TextureCoordinate() {coordinateIndex=1}, False=@n122) {defaultValue=true, group=RedChannel, parameterName=UniversalNoiseUVs_R}
   B=AppendVector(A=@n127.R, B=@n127.R)
  )
  n129 = Multiply(A=AppendVector(A=@n127.B, B=@n127.B), B=NamedRerouteUsage())
  n130 = ChannelMaskParameter(None=TextureSampleParameter2D(UVs=Add(A=@n128, B=@n129)) {group=BaseTextures, parameterName=NoiseMap, texture=&T_NPCity_Noise}.RGBA) {group=RedChannel, parameterName=NoiseChannel_R}
  n131 = StaticSwitchParameter(True=ConstantBiasScale(None=@n130.R) {bias=-0.5, scale=2}, False=@n130.R) {group=RedChannel, parameterName=NoiseCBS_R}
  n132 = Constant()
  n133 = StaticSwitchParameter(True=Multiply(A=Multiply(A=@n126) {constB=0.1}, B=AppendVector(A=@n131, B=@n130.R)), False=@n132) {group=RedChannel, parameterName=Noise_R}
  n134 = NamedRerouteUsage()
  n135 = StaticSwitchParameter(
   True=Multiply(A=ComponentMask(None=@n134) {g=true, r=true}, B=ScalarParameter() {group=RedChannel, parameterName=NormalWarpIntensity_R})
   False=@n132
  ) {group=RedChannel, parameterName=NormalWarp_R}
  n136 = TextureSampleParameter2D(UVs=Add(A=StaticSwitchParameter(True=@n125, False=@n124) {group=RedChannel, parameterName=Rotator_R}, B=Add(A=@n133, B=@n135))) {group=BaseTextures, parameterName=RGBA_Mask, texture=&T_endit_RGB_Fire1}
  n137 = LinearInterpolate(
   B=StaticSwitchParameter(True=OneMinus(None=@n130.R), False=@n130.R) {group=RedChannel, parameterName=NoiseSubtractInvert_R}
   Alpha=ScalarParameter() {defaultValue=0.1, group=RedChannel, parameterName=NoiseSubtractIntensity_R}
  )
  n138 = StaticSwitchParameter(True=Saturate(None=Subtract(A=@n136.R, B=@n137)), False=@n136.R) {group=RedChannel, parameterName=NoiseSubtract_R}
  n139 = StaticSwitchParameter(True=LinearInterpolate(A=@n138, Alpha=NamedRerouteUsage()) {constB=0}, False=@n138) {group=RedChannel, parameterName=RedChannelBehindDiffuse}
  n140 = LinearInterpolate(
   A=LinearInterpolate(A=LinearInterpolate(A=@n52, B=@n61, Alpha=@n81) {constAlpha=1, desc=Diffuse}, B=@n90, Alpha=Saturate(None=@n110)) {constAlpha=1, desc=Diffuse}
   B=@n119
   Alpha=Saturate(None=Multiply(A=@n120, B=@n139))
  ) {constAlpha=1, desc=Diffuse}
  n141 = TextureObjectParameter() {
   desc=RGB +Alpha opacity
   group=UniversalLogos
   parameterName=TrimSheet
   texture=/VehicleCosmetics/UniversalSkins/Psyonix/Skin_Psyonix_TrimSheet
  }
  n142 = VectorParameter() {
   desc="RG=XYScale BA=XYOffset"
   group=UniversalLogos
   parameterName=TrimSheet1_Transform
   channelNames {a=Y Offset, b=X Offset, g=Y Scale, r=X Scale}
   defaultValue=<rgba 1.65 1 0 0.675>
  }
  n143 = ScalarParameter() {defaultValue=1, desc=Additional scaling for autouni logos, group=BaseTextures, parameterName=LogoExtraScaling}
  n144 = Constant() {r=1}
  n145 = MaterialFunctionCall(R=@n143, G=@n143, B=@n144, A=@n144) {materialFunction=&AppendMany}
  n146 = ScalarParameter() {desc=Extra rotation by degrees, group=BaseTextures, parameterName=LogoExtraRotation}
  n147 = VectorParameter() {
   desc="Number of sprites in X,Y=r,g  Select Sprite position starting from top left in X,Y = b,a"
   group=UniversalLogos
   parameterName=TrimSheet1_Cells
   channelNames {a=Sprite Y Position, b=Sprite X Position, g="##Sprites in Y", r="##Sprites in X"}
   defaultValue=<rgba 2 2 2 2>
  }
  n148 = VectorParameter() {
   desc="RG=XY Scale Offset; BA=XY Position Offset"
   group=UniversalLogos
   parameterName=TrimSheet1Mirrored_Offset
   channelNames {a=Y Offset, b=X Offset, g=Y Scale, r=X Scale}
  }
  n149 = MaterialFunctionCall!(
   TrimSheet=@n141
   TrimSheet_Transform=Multiply(A=AppendVector(A=@n142, B=@n142), B=@n145.RGBA)
   TrimSheet_Rotation=Add(A=ScalarParameter() {desc=Degrees counterclockwise, group=UniversalLogos, parameterName=TrimSheet1_Rotation}, B=@n146)
   TrimSheet_Cells=AppendVector(A=@n147."##Sprites in X", B=@n147."##Sprites in X")
   TrimSheet_CellSelector=AppendVector(A=@n147.Sprite X Position, B=@n147.Sprite X Position)
   MirroredTrimSheet_Opacity=ScalarParameter() {desc="Enables the mirrored TrimSheet; 0 = hidden; 1 = visible", group=UniversalLogos, parameterName=TrimSheet1Mirrored_Opacity}
   MirroredTrimSheet_TransformOffset=Multiply(A=@n145.RGBA, B=AppendVector(A=@n148, B=@n148))
   MirroredTrimSheet_RotationOffset=Add(
    A=Multiply(A=@n146) {constB=-1}
    B=ScalarParameter() {desc=Offset by degrees counterclockwise, group=UniversalLogos, parameterName=TrimSheet1Mirrored_RotationOffset}
   )
   MirroredTrimSheet_CellSelector
   MirrorUsesFlippedUVs
  ) {materialFunction=/VehicleCosmetics/AutoUniversalSkins/_MaterialFunctions/MF_AutomatedUniCoordSelector}
  n150 = TextureCoordinate() {coordinateIndex=1}
  n151 = ComponentMask(None=@n150) {g=true}
  n152 = StaticSwitchParameter(True=AppendVector(A=OneMinus(None=@n151), B=ComponentMask(None=@n150) {r=true}), False=@n150) {defaultValue=true, desc=LogoRotate90, group=BaseTextures, parameterName=LogoRotate90}
  n153 = VectorParameter() {group=BaseTextures, parameterName=LogoTransforms, defaultValue=<rgba 13 13 -9.03 -7.5>}
  n154 = AppendVector(A=@n153.R, B=@n153.R)
  n155 = TextureSampleParameter2D(UVs=Add(A=Multiply(A=@n152, B=@n154), B=AppendVector(A=@n153.B, B=@n153.B))) {group=BaseTextures, parameterName=Logo, texture=/VehicleCosmetics/Bodies/Alokin/Skins/AlokinLivery02/T_Pony_Logo}
  n156 = NamedRerouteUsage()
  n157 = NamedRerouteUsage()
  n158 = StaticSwitchParameter(
   True=VectorParameter() {group=BaseTextures, parameterName=Logo_UniqueColor, defaultValue=<rgba 1 1 0 1>}.RGB
   False=StaticSwitchParameter(True=NamedRerouteUsage(), False=StaticSwitchParameter(True=@n156, False=@n157) {group=BaseTextures, parameterName=Logo_UseCustomColor}) {group=BaseTextures, parameterName=Logo_UsePaintColor}
  ) {group=BaseTextures, parameterName=Logo_UseUniqueColor}
  n159 = Multiply(A=TextureCoordinate() {coordinateIndex=1}, B=ScalarParameter() {defaultValue=1, group=BaseTextures, parameterName=Logo_MultichromeScale})
  n160 = Multiply(A=ScalarParameter() {defaultValue=0.2, group=BaseTextures, parameterName=Logo_MultichromePanSpeed}, B=NamedRerouteUsage())
  n161 = MaterialFunctionCall(Hue Shift Percentage=ComponentMask(None=Panner(Coordinate=@n159, Time=@n160) {speedY=1}) {g=true}, Texture=@n158) {materialFunction=&HueShift}
  n162 = ScalarParameter() {defaultValue=1, group=BaseTextures, parameterName=Logo_MultichromeColor}
  n163 = StaticSwitchParameter(
   True=StaticSwitchParameter(True=ComponentMask(None=@n149.TextureOutput) {b=true, g=true, r=true}, False=@n155.RGB) {group=BaseTextures, parameterName=LogoUseAutoUni}
   False=StaticSwitchParameter(True=LinearInterpolate(A=@n158, B=@n161.Result, Alpha=@n162), False=@n158) {group=BaseTextures, parameterName=Logo_UseMultichromeColor}
  ) {defaultValue=true, group=BaseTextures, parameterName=LogoUseTextureColor}
  n164 = StaticSwitchParameter(True=ComponentMask(None=@n149.TextureOutput) {a=true}, False=@n155.A) {group=BaseTextures, parameterName=LogoUseAutoUni}
  n165 = StaticSwitchParameter(True=LinearInterpolate(A=@n140, B=@n163, Alpha=@n164), False=@n140) {group=BaseTextures, parameterName=Logo}
  n166 = NamedRerouteUsage()
  n167 = Transform(None=CameraVectorWS()) {transformSourceType=TRANSFORMSOURCE_World, transformType=TRANSFORM_Tangent}
  n168 = NamedRerouteUsage()
  n169 = Power(Base=OneMinus(None=DotProduct(A=@n167, B=@n168))) {constExponent=3}
  n170 = OneMinus(None=@n169)
  n171 = ScalarParameter() {defaultValue=1, group=Finish, parameterName=FinishInteriorFalloff}
  n172 = ScalarParameter() {defaultValue=1, group=Finish, parameterName=FinishOuterFalloff}
  n173 = Max(A=Saturate(None=Power(Base=@n170, Exp=@n171)), B=Saturate(None=Power(Base=@n169, Exp=@n172)))
  n174 = ScalarParameter() {defaultValue=95, group=Finish, parameterName=MetallicFinishInteriorFalloff}
  n175 = Clamp(None=Power(Base=OneMinus(None=@n169), Exp=@n174), Min=ScalarParameter() {defaultValue=0.1, group=Finish, parameterName=MetallicDarkMinLevel})
  n176 = ScalarParameter() {defaultValue=5, group=Finish, parameterName=MetallicFinishOuterFalloff}
  n177 = Max(A=@n175, B=Saturate(None=Power(Base=@n169, Exp=@n176)))
  n178 = Power(Base=@n175, Exp=@n174)
  n179 = ScalarParameter() {defaultValue=0.1, group=Finish, parameterName=MetallicHighlight}
  n180 = LinearInterpolate(
   A=Multiply(A=@n165, B=Multiply(A=@n166, B=@n173))
   B=Add(A=Multiply(A=@n165, B=@n177), B=Multiply(A=@n178, B=@n179))
   Alpha=NamedRerouteUsage()
  )
  NamedRerouteDeclaration(None=Saturate(None=ConstantBiasScale(None=@n180) {bias=-0.01, desc=Color correction for BR, scale=1.01})) {name=BaseColor, nodeColor=<rgba 1 0 0.375 1>}
  n181 = Sine(None=Multiply(A=ScalarParameter() {defaultValue=1, group=Fresnel, parameterName=EmissivePulseTiming_Fresnel}, B=NamedRerouteUsage())) {period=3}
  n182 = ScalarParameter() {defaultValue=0.2, group=Fresnel, parameterName=EmissiveIntensity_Fresnel}
  n183 = Multiply(
   A=LinearInterpolate(B=ConstantBiasScale(None=@n181) {bias=1.25}, Alpha=ScalarParameter() {defaultValue=1, group=Fresnel, parameterName=PulseInfluence_Fresnel}) {constA=1}
   B=@n182
  )
  n184 = LinearInterpolate(
   A=Multiply(A=@n9, B=ScalarParameter() {group=Colors, parameterName=DiffuseEmissiveIntensity}) {desc=Emissive}
   B=@n19
   Alpha=Multiply(A=@n21, B=StaticSwitchParameter(True=@n183, False=@n182) {defaultValue=true, group=Fresnel, parameterName=EmissivePulse_Fresnel})
  ) {constAlpha=1, desc=Emissive}
  n185 = NamedRerouteUsage()
  n186 = ScalarParameter() {defaultValue=3, group=AlphaChannel, parameterName=EmissivePulseTiming_A}
  n187 = ConstantBiasScale(None=Sine(None=Multiply(A=@n185, B=@n186)) {period=3}) {bias=1.25}
  n188 = ScalarParameter() {defaultValue=0.25, group=AlphaChannel, parameterName=PulseInfluence_A}
  n189 = ScalarParameter() {defaultValue=0.2, group=AlphaChannel, parameterName=EmissiveIntensity_A}
  n190 = StaticSwitchParameter(True=Multiply(A=LinearInterpolate(B=@n187, Alpha=@n188) {constA=1}, B=@n189), False=@n189) {defaultValue=true, group=AlphaChannel, parameterName=EmissivePulse_A}
  n191 = Multiply(A=@n31, B=@n190)
  n192 = Saturate(None=Multiply(A=@n51, B=@n190))
  n193 = StaticSwitchParameter(
   True=StaticSwitchParameter(True=LinearInterpolate(A=@n184, B=@n191, Alpha=@n51), False=LinearInterpolate(A=@n184, B=@n31, Alpha=@n192) {constAlpha=1, desc=Emissive}) {desc="T=Clean F=Additive, effect is subtle", group=Colors, parameterName=EmissiveBlendMode}
   False=@n184
  ) {defaultValue=true, group=AlphaChannel, parameterName=Emissive_A}
  n194 = NamedRerouteUsage()
  n195 = ScalarParameter() {defaultValue=-1, group=BlueChannel, parameterName=EmissivePulseTiming_B}
  n196 = ConstantBiasScale(None=Sine(None=Multiply(A=@n194, B=@n195)) {period=3}) {bias=1.25}
  n197 = ScalarParameter() {defaultValue=1, group=BlueChannel, parameterName=PulseInfluence_B}
  n198 = ScalarParameter() {defaultValue=0.5, group=BlueChannel, parameterName=EmissiveIntensity_B}
  n199 = StaticSwitchParameter(True=Multiply(A=LinearInterpolate(B=@n196, Alpha=@n197) {constA=1}, B=@n198), False=@n198) {defaultValue=true, group=BlueChannel, parameterName=EmissivePulse_B}
  n200 = Multiply(A=@n61, B=@n199)
  n201 = Saturate(None=Multiply(A=@n80, B=@n199))
  n202 = StaticSwitchParameter(
   True=StaticSwitchParameter(True=LinearInterpolate(A=@n193, B=@n200, Alpha=@n80), False=LinearInterpolate(A=@n193, B=@n61, Alpha=@n201) {constAlpha=1, desc=Emissive}) {desc="T=Clean F=Additive, effect is subtle", group=Colors, parameterName=EmissiveBlendMode}
   False=@n193
  ) {defaultValue=true, group=BlueChannel, parameterName=Emissive_B}
  n203 = NamedRerouteUsage()
  n204 = ScalarParameter() {defaultValue=3, group=GreenChannel, parameterName=EmissivePulseTiming_G}
  n205 = ConstantBiasScale(None=Sine(None=Multiply(A=@n203, B=@n204)) {period=3}) {bias=1.25}
  n206 = ScalarParameter() {defaultValue=0.25, group=GreenChannel, parameterName=PulseInfluence_G}
  n207 = ScalarParameter() {defaultValue=3, group=GreenChannel, parameterName=EmissiveIntensity_G}
  n208 = StaticSwitchParameter(True=Multiply(A=LinearInterpolate(B=@n205, Alpha=@n206) {constA=1}, B=@n207), False=@n207) {defaultValue=true, group=GreenChannel, parameterName=EmissivePulse_G}
  n209 = Multiply(A=@n90, B=@n208)
  n210 = Saturate(None=Multiply(A=@n109, B=@n208))
  n211 = StaticSwitchParameter(
   True=StaticSwitchParameter(True=LinearInterpolate(A=@n202, B=@n209, Alpha=@n109), False=LinearInterpolate(A=@n202, B=@n90, Alpha=@n210) {constAlpha=1, desc=Emissive}) {desc="T=Clean F=Additive, effect is subtle", group=Colors, parameterName=EmissiveBlendMode}
   False=@n202
  ) {defaultValue=true, group=GreenChannel, parameterName=Emissive_G}
  n212 = NamedRerouteUsage()
  n213 = ScalarParameter() {defaultValue=1, group=RedChannel, parameterName=EmissivePulseTiming_R}
  n214 = ConstantBiasScale(None=Sine(None=Multiply(A=@n212, B=@n213)) {period=3}) {bias=1.25}
  n215 = ScalarParameter() {defaultValue=1, group=RedChannel, parameterName=PulseInfluence_R}
  n216 = ScalarParameter() {defaultValue=3, group=RedChannel, parameterName=EmissiveIntensity_R}
  n217 = StaticSwitchParameter(True=Multiply(A=LinearInterpolate(B=@n214, Alpha=@n215) {constA=1}, B=@n216), False=@n216) {defaultValue=true, group=RedChannel, parameterName=EmissivePulse_R}
  n218 = Multiply(A=@n119, B=@n217)
  n219 = Saturate(None=Multiply(A=@n139, B=@n217))
  n220 = StaticSwitchParameter(
   True=StaticSwitchParameter(True=LinearInterpolate(A=@n211, B=@n218, Alpha=@n139), False=LinearInterpolate(A=@n211, B=@n119, Alpha=@n219) {constAlpha=1, desc=Emissive}) {desc="T=Clean F=Additive, effect is subtle", group=Colors, parameterName=EmissiveBlendMode}
   False=@n211
  ) {defaultValue=true, group=RedChannel, parameterName=Emissive_R}
  n221 = Multiply(A=ScalarParameter() {group=BaseTextures, parameterName=LogoEmissive}, B=@n163)
  n222 = Saturate(None=StaticSwitchParameter(True=LinearInterpolate(A=@n220, B=@n221, Alpha=@n164), False=@n220) {group=BaseTextures, parameterName=Logo})
  n223 = ScalarParameter() {defaultValue=3, group=Colors, parameterName=EmissiveIntensity}
  NamedRerouteDeclaration(None=ConstantBiasScale(None=Multiply(A=@n222, B=@n223)) {bias=-0.015, desc=Color correction for BR, scale=1.15}) {name=EmissiveColor, nodeColor=<rgba 0 0.6953 1 1>}
  n224 = StaticSwitchParameter(True=NamedRerouteUsage(), False=TextureCoordinate()) {
   defaultValue=true
   desc="Normal and AO/Spec/Metallic/Roughness map coordinates. Universal=True, body specific=false"
   group=BaseTextures
   parameterName=UniversalUVs_Normals
  }
  n225 = TextureSampleParameter2D(UVs=@n224) {
   group=BaseTextures
   parameterName=NormalMap
   samplerSource=SSM_Wrap_WorldGroupSettings
   samplerType=SAMPLERTYPE_Normal
   texture=&/T_endit_N_Fire1
  }
  n226 = Constant3Vector() {constant=<rgba 0 0 1 0>}
  n227 = StaticSwitchParameter(True=@n225.RGB, False=@n226) {defaultValue=true, group=BaseTextures, parameterName=NormalMap}
  NamedRerouteDeclaration(None=StaticSwitchParameter(True=LinearInterpolate(A=@n227, B=@n226, Alpha=@n164), False=@n227) {group=BaseTextures, parameterName=Logo}) {name=Normals, nodeColor=<rgba 0.25 0.25 1 1>}
  n228 = TextureCoordinate() {coordinateIndex=1}
  n229 = VectorParameter() {desc="(R,G) = Detail1    (B,A) = Detail2", group=BaseTextures, parameterName=Universal_Map_Transform, defaultValue=<rgba 1 1 0 0>}
  n230 = ComponentMask(None=@n229.RGB) {g=true, r=true}
  NamedRerouteDeclaration(None=Add(A=Multiply(A=@n228, B=@n230), B=AppendVector(A=@n229.B, B=@n229.B)) {desc=Panner}) {name=DiffuseNormal UVs, nodeColor=<rgba 0.1484 0 1 1>}
  n231 = VectorParameter() {desc="Alpha = Emissive Color Brightness", group=Colors, parameterName=BaseColor, defaultValue=<rgba 1 0.0648 0 1>}
  NamedRerouteDeclaration(None=@n231.RGB) {name=TeamColor, nodeColor=<rgba 0 0.5312 1 1>}
  NamedRerouteDeclaration(None=VectorParameter() {group=Colors, parameterName=AccentColor, defaultValue=<rgba 1 0.3005 0.0262 1>}.RGB) {name=CustomColor, nodeColor=<rgba 0 1 0.0625 1>}
  NamedRerouteDeclaration(None=VectorParameter() {group=Colors, parameterName=PaintColor, defaultValue=<rgba 0.0137 0.0137 0.0137 1>}.RGB) {name=PaintColor, nodeColor=<rgba 0 0 1 1>}
  NamedRerouteDeclaration(None=Time() {bOverride_Period=true, period=500}) {name=time, nodeColor=<rgba 1 0 0.4219 1>}
  n232 = TextureSampleParameter2D(UVs=@n224) {
   desc="R=specular, G=ao, B=Roughness, A = Metallic"
   group=BaseTextures
   parameterName=AmbientOcclusion
   texture=&/T_endit_AO_Fire
  }
  n233 = ScalarParameter() {defaultValue=0.5, group=Finish, parameterName=Roughness}
  n234 = Saturate(None=StaticSwitchParameter(True=@n232.B, False=@n233) {desc=Roughness in B channel of AO map, group=BaseTextures, parameterName=Rough Map})
  NamedRerouteDeclaration(None=StaticSwitchParameter(True=LinearInterpolate(A=@n234, B=@n233, Alpha=@n164), False=@n234) {group=BaseTextures, parameterName=Logo}) {name=Roughness, nodeColor=<rgba 0 0.1094 1 1>}
  n235 = ScalarParameter() {group=Finish, parameterName=MetallicDefault}
  n236 = StaticSwitchParameter(True=@n232.A, False=@n235) {desc=Metallic in A channel of AO map, group=BaseTextures, parameterName=Metallic Map}
  NamedRerouteDeclaration(None=StaticSwitchParameter(True=LinearInterpolate(A=@n236, B=@n235, Alpha=@n164), False=@n236) {group=BaseTextures, parameterName=Logo}) {name=Metallic, nodeColor=<rgba 0.4297 1 0 1>}
  n237 = ScalarParameter() {defaultValue=1, group=Finish, parameterName=SpecularHighlightIntensity}
  n238 = StaticSwitchParameter(True=@n232.R, False=@n237) {desc="Spec in R channel of AO map (helps Finish consistency)", group=BaseTextures, parameterName=SpecMap}
  NamedRerouteDeclaration(None=StaticSwitchParameter(True=LinearInterpolate(A=@n238, B=@n237, Alpha=@n164), False=@n238) {group=BaseTextures, parameterName=Logo}) {name=Specular, nodeColor=<rgba 1 0.2578 0 1>}
  n239 = Constant3Vector() {constant=<rgba 1 1 1 0>}
  n240 = StaticSwitchParameter(True=LinearInterpolate(A=@n239, B=@n232.G, Alpha=ScalarParameter() {defaultValue=1, group=BaseTextures, parameterName=AO}), False=@n239) {defaultValue=true, desc=AO in G channel of AO map, group=BaseTextures, parameterName=AmbientOcclusionMap}
  NamedRerouteDeclaration(None=StaticSwitchParameter(True=LinearInterpolate(A=@n240, B=@n239, Alpha=@n164), False=@n240) {group=BaseTextures, parameterName=Logo}) {name=AO, nodeColor=<rgba 0.7969 0.7969 0.7969 1>}
 ]}
